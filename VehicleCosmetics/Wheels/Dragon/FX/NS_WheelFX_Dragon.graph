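# Niagara system for the Dragon vehicle wheel cosmetic, emitting short-lived, glowing orange particles that drift upward from each wheel.
# The active emitter partitions spawns by wheel count, samples the skeletal mesh per WheelID, and positions particles through VC_WheelXform.
# Curves scale size and color over life, while time-offset dynamic material parameters feed mesh and sprite renderers.
# A similar WheelFX_Painted emitter, colored from a painted-variant palette, is disabled.
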NiagaraSystem /VehicleCosmetics/Wheels/Dragon/FX/NS_WheelFX_Dragon
dependencies {
 usedDataInterfaces [
  &NiagaraDataInterfaceCurve
  &NiagaraDataInterfaceColorCurve
  &NiagaraDataInterfaceVectorCurve
  &NiagaraDataInterfaceSkeletalMesh
 ]
 usedDynamicInputs [&VectorFromCurve]
 usedModules [
  &self:SystemSpawnScript.SystemScriptSource.SystemScriptGraph.NiagaraNodeAssignment_0.SetVariables_#1
  &FN_UnpackVehicleFlags
  /VehicleCosmetics/FX/Niagara/Modules/WheelFX/VC_WheelFX_SystemSetup
  /VehicleCosmetics/FX/Niagara/Modules/VC_SetVehicleSpeedAttribs
  &SystemState
 ]
}
emitters {
 WheelFX_Painted {
  bEnabled=false
  emitterSpawnScript [/VehicleCosmetics/FX/Niagara/Modules/WheelFX/VC_WheelFX_EmitterSetup()]
  emitterUpdateScript [
   &EmitterState(Life Cycle Mode=::NewEnumerator1, Loop Behavior=::NewEnumerator1, Loop Duration Mode=::NewEnumerator1, UseLoopDelay=false)
   &SpawnRate(SpawnRate=200, Use Spawn Probability=false) {module=SpawnRate001}
   &SpawnBurst_Instantaneous(Spawn Count=$Emitter.NumWheels, Use Spawn Probability=false, UseLoopCountLimit=false) {enabled=false}
   &SpawnRate(SpawnRate=15, Use Spawn Probability=false) {enabled=false}
  ]
  particleSpawnScript [
   &PartitionParticles(Number Of Partitions=$Emitter.NumWheels, Partition Mode:ENiagara_PartitionMode_v2=::NewEnumerator5)
   /VehicleCosmetics/FX/Niagara/Modules/WheelFX/VC_WheelFX_ParticleSetup(WheelID:NiagaraInt32=$Particles.PartitionParticles.Partition)
   &VC_PaintedVariantsOverride(
    Black=<rgba 0.01 0.01 0.01 500>
    BurntSienna=<rgba 0.05 0.025 0.0125 500>
    Cobalt=<rgba 0 0.037 2 250>
    Crimson=<rgba 2 0 0.0542 250>
    DefaultColor=<rgba 1 0.4375 0.1 250>
    ForestGreen=<rgba 0 1 0.0505 250>
    Gold:LinearColor=<rgba 4 1.8477 0 100>
    Grey=<rgba 0.03 0.03 0.03 250>
    Lime=<rgba 0.55 1 0.1 250>
    Orange=<rgba 1.5 0.5232 0 250>
    Pink=<rgba 1 0.1198 0.6799 250>
    Purple=<rgba 0.6156 0 1.5 250>
    Saffron=<rgba 0.9 0.69 0 250>
    SkyBlue=<rgba 0 1 1 250>
    TitaniumWhite=<rgba 0.6 0.75 0.75 250>
   )
   &InitializeParticle(
    Color=$Particles.Color
    Lifetime Max=0.125
    Lifetime Min=0.25
    Lifetime Mode=::NewEnumerator1
    Position=<0 12 -2.5>
    Position Mode=::NewEnumerator1
    Sprite Rotation Mode=::NewEnumerator2
    Sprite Size Max=<10 15>
    Sprite Size Min=<5 10>
    Sprite Size Mode=::NewEnumerator2
    UsePositionOffset=false
    Write Material Random=true
   )
   &ShapeLocation(
    MaskSpawnGroup=false
    Override Seed=false
    Ring Radius=&Multiply_Float(A=$Particles.WheelRadius, B=0)
    Rotation Angle=90
    Rotation Angle Type:ENiagara_AngleInput=::NewEnumerator0
    Rotation Axis=<0 1 0>
    Rotation Mode=::NewEnumerator1
    Shape Primitive=::NewEnumerator4
    UseOwnerScale=true
   ) {enabled=false}
   &AddVelocity(
    Cone Angle=10
    Cone Axis=<1 0.25 -1>
    Module.UseConeFalloff=false
    Module.UseVelDistribution=false
    Override Seed=false
    Rotation Coordinate Space=::Simulation
    Velocity Mode=::NewEnumerator2
    Velocity Speed=80
   )
   /SetVariables(Particles.glowColor:LinearColor=&MultiplyLinearColor(Linear Color=$Particles.Color, Scale Factor=<1 0.5 0.5 0.0008>)) {isSetParameters=true, module=SetVariables_#2}
  ]
  particleUpdateScript [
   &ParticleState()
   &ScaleSpriteSize(Uniform Curve Sprite Scale=DataInterface() {propertyValues {Curve {
       defaultValue=0
       keys [
        {arriveTangent=1.6418, interpMode=RCIM_Cubic, leaveTangent=1.6418, tangentMode=RCTM_User, time=0, value=0.4}
        {tangentMode=RCTM_User, time=0.8, value=1}
       ]
      }}})
   &ScaleColor(
    Linear Color Curve=DataInterface() {propertyValues {
      AlphaCurve {keys [<at 0 1>,<at 1 0.2>]}
      BlueCurve {keys [<at 0 7>,<at 0.8716 0.1>]}
      GreenCurve {keys [<at 0 7>,<at 0.8716 0.1>]}
      RedCurve {keys [<at 0 7>,<at 0.8716 0.5>]}
     }}
    Scale Mode=2
   )
   &GravityForce(Gravity=<0 0 200>)
   &SolveForcesAndVelocity(Clamp Velocity=false, Limit Acceleration=false)
   &SkeletalMeshLocation(
    Bone / Socket Index=$Particles.WheelID
    Bone / Socket Sampling Mode=::NewEnumerator5
    MaskSpawnGroup=false
    Mesh Sampling Type=::NewEnumerator6
    Orientation Sampling=::NewEnumerator1
    Position Sampling=::NewEnumerator2
    Skeletal Mesh=$System.SkeletalMesh
   )
   /VehicleCosmetics/FX/Niagara/Modules/WheelFX/VC_WheelXform(
    SKM Position:NiagaraPosition=$Particles.SkeletalMeshLocation.SampledPosition
    SKM Rotation:Quat4f=$Particles.SkeletalMeshLocation.SampledRotation
    SKM Velocity:Vector3f=$Particles.SkeletalMeshLocation.SampledVelocity
   )
   &DynamicMaterialParameters(
    Index 0 Param 1=0.5
    Index 0 Param 2=0.1
    Index 0 Param 3=&Multiply_Float(
     A=0.5
     B=&FloatFromCurve(FloatCurve=DataInterface() {propertyValues {Curve {
         defaultValue=1
         keys [
          {arriveTangent=0.7065, interpMode=RCIM_Cubic, leaveTangent=0.7065, tangentMode=RCTM_User, time=0, value=0.5}
          {arriveTangent=0.0224, interpMode=RCIM_Cubic, leaveTangent=0.0224, tangentMode=RCTM_User, time=1, value=1.5}
         ]
        }}})
    )
    Index 1 Param 1=1
    Index 1 Param 2=1
    Index 2 Param 1=0.1
    Index 2 Param 2=0.1
    Index 2 Param 3=&Add_Float(
     A=&RandomRangeFloat(Minimum=-100, Maximum=100)
     B=&Multiply_Float(A=1, B=$Engine.Time)
    )
    Module.Index0Float0Write=true
    Module.Index0Float1Write=true
    Module.Index0Float2Write=true
    Module.Index0Float3Write=true
    Module.Index1Float0Write=true
    Module.Index1Float1Write=true
    Module.Index1Float2Write=true
    Module.Index1Float3Write=true
    Module.Index2Float0Write=true
    Module.Index2Float1Write=true
    Module.Index2Float2Write=true
    Module.Index2Float3Write=true
    Write Parameter Index 1=::NewEnumerator1
    Write Parameter Index 2=::NewEnumerator1
   )
   /SetVariables(Particles.glowSIze:Vector2f=&Multiply_Vector2D_ByFloat(Float=1.5, Vector2D=$Particles.SpriteSize)) {isSetParameters=true, module=SetVariables_#3}
  ]
  rendererClasses [&NiagaraMeshRendererProperties, &NiagaraSpriteRendererProperties]
  renderers [
   {rendererClass=&NiagaraMeshRendererProperties, rendererIndex=0}
   {rendererClass=&NiagaraSpriteRendererProperties, rendererIndex=1}
   {rendererClass=&NiagaraSpriteRendererProperties, rendererIndex=2}
  ]
 }
 WheelFX_Sprites001 {
  emitterSpawnScript [&VC_WheelFX_EmitterSetup()]
  emitterUpdateScript [
   &EmitterState(Life Cycle Mode=::NewEnumerator1, Loop Behavior=::NewEnumerator1, Loop Duration Mode=::NewEnumerator1, UseLoopDelay=false)
   &SpawnRate(SpawnRate=220, Use Spawn Probability=false) {module=SpawnRate001}
   &SpawnBurst_Instantaneous(Spawn Count=$Emitter.NumWheels, Use Spawn Probability=false, UseLoopCountLimit=false) {enabled=false}
   &SpawnRate(SpawnRate=15, Use Spawn Probability=false) {enabled=false}
  ]
  particleSpawnScript [
   &PartitionParticles(Number Of Partitions=$Emitter.NumWheels, Partition Mode:ENiagara_PartitionMode_v2=::NewEnumerator5)
   &VC_WheelFX_ParticleSetup(WheelID:NiagaraInt32=$Particles.PartitionParticles.Partition)
   &VC_PaintedVariantsOverride()
   &InitializeParticle(
    Color=<rgba 1.5 0.5951 0.2205 250>
    Lifetime Max=0.25
    Lifetime Min=0.1
    Lifetime Mode=::NewEnumerator1
    Position=<0 25 -3>
    Position Mode=::NewEnumerator1
    Sprite Rotation Mode=::NewEnumerator2
    Sprite Size Max=<12 20>
    Sprite Size Min=<5 10>
    Sprite Size Mode=::NewEnumerator2
    UsePositionOffset=false
    Write Material Random=true
   )
   &ShapeLocation(
    MaskSpawnGroup=false
    Override Seed=false
    Ring Radius=&Multiply_Float(A=$Particles.WheelRadius, B=0)
    Rotation Angle=90
    Rotation Angle Type:ENiagara_AngleInput=::NewEnumerator0
    Rotation Axis=<0 1 0>
    Rotation Mode=::NewEnumerator1
    Shape Primitive=::NewEnumerator4
    UseOwnerScale=true
   )
   &AddVelocity(
    Cone Angle=10
    Cone Axis=<1 0.25 -1>
    Module.UseConeFalloff=false
    Module.UseVelDistribution=false
    Override Seed=false
    Rotation Coordinate Space=::Simulation
    Velocity Mode=::NewEnumerator2
    Velocity Speed=80
   )
   /SetVariables(Particles.glowColor:LinearColor=<rgba 1 0.15 0 0.2>) {isSetParameters=true, module=SetVariables_#2}
  ]
  particleUpdateScript [
   &ParticleState()
   &ScaleSpriteSize(Uniform Curve Sprite Scale=DataInterface() {propertyValues {Curve {
       defaultValue=0
       keys [
        {arriveTangent=1.0401, interpMode=RCIM_Cubic, leaveTangent=1.0401, tangentMode=RCTM_User, time=0, value=0.25}
        {tangentMode=RCTM_User, time=0.8, value=1}
       ]
      }}})
   &ScaleColor(
    Linear Color Curve=DataInterface() {propertyValues {
      AlphaCurve {keys [<at 0 1>,<at 1 0>]}
      BlueCurve {keys [<at 0 10>,<at 0.8668 0>]}
      GreenCurve {keys [<at 0 12.9277>,<at 0.8668 0.15>]}
      RedCurve {keys [<at 0 20>,<at 0.8668 3>]}
     }}
    Scale Mode=2
   )
   &GravityForce(Gravity=<0 0 30>)
   &SolveForcesAndVelocity(Clamp Velocity=false, Limit Acceleration=false)
   &SkeletalMeshLocation(
    Bone / Socket Index=$Particles.WheelID
    Bone / Socket Sampling Mode=::NewEnumerator5
    MaskSpawnGroup=false
    Mesh Sampling Type=::NewEnumerator6
    Orientation Sampling=::NewEnumerator1
    Position Sampling=::NewEnumerator2
    Skeletal Mesh=$System.SkeletalMesh
   )
   &VC_WheelXform(
    SKM Position:NiagaraPosition=$Particles.SkeletalMeshLocation.SampledPosition
    SKM Rotation:Quat4f=$Particles.SkeletalMeshLocation.SampledRotation
    SKM Velocity:Vector3f=$Particles.SkeletalMeshLocation.SampledVelocity
   )
   &DynamicMaterialParameters(
    Index 0 Param 1=0.5
    Index 0 Param 2=1
    Index 0 Param 3=&Multiply_Float(
     A=0.4
     B=&FloatFromCurve(FloatCurve=DataInterface() {propertyValues {Curve {
         defaultValue=1
         keys [
          {arriveTangent=0.7065, interpMode=RCIM_Cubic, leaveTangent=0.7065, tangentMode=RCTM_User, time=0, value=0.5}
          {arriveTangent=0.0224, interpMode=RCIM_Cubic, leaveTangent=0.0224, tangentMode=RCTM_User, time=1, value=1.5}
         ]
        }}})
    )
    Index 1 Param 1=1
    Index 1 Param 2=1
    Index 2 Param 1=0.1
    Index 2 Param 2=0.1
    Index 2 Param 3=&Add_Float(
     A=&RandomRangeFloat(Minimum=-100, Maximum=100)
     B=&Multiply_Float(A=1, B=$Engine.Time)
    )
    Module.Index0Float0Write=true
    Module.Index0Float1Write=true
    Module.Index0Float2Write=true
    Module.Index0Float3Write=true
    Module.Index1Float0Write=true
    Module.Index1Float1Write=true
    Module.Index1Float2Write=true
    Module.Index1Float3Write=true
    Module.Index2Float0Write=true
    Module.Index2Float1Write=true
    Module.Index2Float2Write=true
    Module.Index2Float3Write=true
    Write Parameter Index 1=::NewEnumerator1
    Write Parameter Index 2=::NewEnumerator1
   )
   /SetVariables(Particles.glowSIze:Vector2f=&Multiply_Vector2D_ByFloat(Float=1.5, Vector2D=$Particles.SpriteSize)) {isSetParameters=true, module=SetVariables_#3}
  ]
  rendererClasses [&NiagaraMeshRendererProperties, &NiagaraSpriteRendererProperties]
  renderers [
   {rendererClass=&NiagaraMeshRendererProperties, rendererIndex=0}
   {rendererClass=&NiagaraSpriteRendererProperties, rendererIndex=1}
   {rendererClass=&NiagaraSpriteRendererProperties, rendererIndex=2}
  ]
 }
}
summary {
 systemName=NS_WheelFX_Dragon
 emitters [
  {
   bEnabled=false
   emitterName=WheelFX_Painted
   rendererClasses [&NiagaraMeshRendererProperties, &NiagaraSpriteRendererProperties]
  }
  {
   emitterName=WheelFX_Sprites001
   rendererClasses [&NiagaraMeshRendererProperties, &NiagaraSpriteRendererProperties]
  }
 ]
 userVariables [
  User.bIsFrontendPreview {type=NiagaraBool}
  User.bIsPainted {type=NiagaraBool}
  User.ColorVariantID {type=NiagaraInt32}
  User.PaintedColor {type=LinearColor, description="", defaultValue {struct=LinearColor, value=<rgba 1 1 1 1>}}
  User.PaintedScalar {type=NiagaraFloat, defaultValue=0, description=""}
  User.VehicleFlags {type=NiagaraInt32, defaultValue=806289408}
  User.Wheel_Radius_Array {
   type=NiagaraDataInterfaceArrayFloat
   defaultValue {
    struct=NiagaraExt_VariableValue_DataInterface
    value {
     dataInterface=&self:NiagaraDataInterfaceArrayFloat_1
     dataInterfaceClass=&NiagaraDataInterfaceArrayFloat
    }
   }
  }
  User.Wheel_Width_Array {
   type=NiagaraDataInterfaceArrayFloat
   defaultValue {
    struct=NiagaraExt_VariableValue_DataInterface
    value {
     dataInterface=&self:NiagaraDataInterfaceArrayFloat_0
     dataInterfaceClass=&NiagaraDataInterfaceArrayFloat
    }
   }
  }
  User.WheelTouchingArray {
   type=NiagaraDataInterfaceArrayFloat
   defaultValue {
    struct=NiagaraExt_VariableValue_DataInterface
    value {
     dataInterface=&self:NiagaraDataInterfaceArrayFloat_2
     dataInterfaceClass=&NiagaraDataInterfaceArrayFloat
    }
   }
  }
 ]
}
