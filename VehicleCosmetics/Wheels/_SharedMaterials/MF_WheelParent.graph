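# Shared parent material function for cosmetic vehicle wheels, outputting material attributes for both rim and tire.
# Channel-picked masks tint each part with custom and accent colors, with optional inverted, floored or flip-flop masking, then multiply by AO.
# Lerps merge rim and tire color, normal, emissive and metallic/specular/roughness, the latter from PBR textures or scalars.
# Tread UVs can flip with corrected normals, and vertex-color green masks TirePop world position offset.
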MaterialFunction /VehicleCosmetics/Wheels/_SharedMaterials/MF_WheelParent
graph {nodes [
  n1 = TextureSampleParameter2D() {group=Rim Color Control, parameterName=Rim_CustomColorMask, texture=/VehicleCosmetics/Textures/T_Black}
  n2 = ChannelMaskParameter(None=@n1.RGBA) {group=Rim Color Control, maskChannel=Green, parameterName=Rim_AccentColorChannelPicker, defaultValue=<rgba 0 1 0 0>}
  n3 = ScalarParameter() {defaultValue=1, group=Rim Color Control, parameterName=Rim_PaintedAccentFloorValue}
  Composite!(InputChannel=@n2.R, FloorValue=@n3) {subgraphName=RimFloorAccentColor}
  n4 = TextureSampleParameter2D(UVs=NamedRerouteUsage()) {group=Tire Color Control, parameterName=Tire_CustomColorMask, texture=&T_Black}
  n5 = ChannelMaskParameter(None=@n4.RGBA) {group=Tire Color Control, parameterName=Tire_CustomColorChannelPicker, sortPriority=2}
  n6 = ScalarParameter() {group=Tire Color Control, parameterName=Tire_PaintedCustomFloorValue}
  Composite(InputChannel=@n5.R, FloorValue=@n6) {subgraphName=FloorCustomColor}
  Composite(InputChannel=@n5.R, FloorValue=@n6) {subgraphName=FloorCustomColorFlip}
  n7 = ChannelMaskParameter(None=@n4.RGBA) {group=Tire Color Control, maskChannel=Green, parameterName=Tire_AccentColorChannelPicker, defaultValue=<rgba 0 1 0 0>}
  n8 = ScalarParameter() {group=Tire Color Control, parameterName=Tire_PaintedAccentFloorValue}
  Composite(InputChannel=@n7.R, FloorValue=@n8) {subgraphName=FloorAccentColor}
  Composite(InputChannel=@n2.R, FloorValue=@n3) {subgraphName=RimFloorAccentColorFlip}
  Composite(InputChannel=@n7.R, FloorValue=@n8) {subgraphName=FloorAccentColorFlip}
  n9 = ChannelMaskParameter(None=@n1.RGBA) {group=Rim Color Control, parameterName=Rim_CustomColorChannelPicker, sortPriority=2}
  n10 = ScalarParameter() {group=Rim Properties, parameterName=Rim_PaintedCustomFloorValue}
  Composite(InputChannel=@n9.R, FloorValue=@n10) {subgraphName=RimFloorCustomColorFlip}
  Composite(InputChannel=@n9.R, FloorValue=@n10) {subgraphName=RimFloorCustomColor}
  MaterialExpressionReroute_22 = Reroute(None=@n2.R)
  MaterialExpressionReroute_23 = Reroute(None=@n3)
  n11 = Add(A=@MaterialExpressionReroute_23.FloorValue) {constB=1.5}
  Floor(None=Saturate(None=Multiply(A=@MaterialExpressionReroute_22.InputChannel, B=@n11) {constB=1.5}))
  n12 = Constant3Vector()
  n13 = Add(A=Constant()) {constB=1.5}
  Floor(None=Saturate(None=Multiply(A=@n12, B=@n13) {constB=1.5}))
  MaterialExpressionReroute_15 = Reroute(None=@n5.R)
  MaterialExpressionReroute_16 = Reroute(None=@n6)
  n14 = Add(A=@MaterialExpressionReroute_16.FloorValue) {constB=1.5}
  Floor(None=Saturate(None=Multiply(A=@MaterialExpressionReroute_15.InputChannel, B=@n14) {constB=1.5}))
  MaterialExpressionReroute_21 = Reroute(None=@n7.R)
  MaterialExpressionReroute_25 = Reroute(None=@n8)
  n15 = Add(A=@MaterialExpressionReroute_25.FloorValue) {constB=1.5}
  Floor(None=Saturate(None=Multiply(A=@MaterialExpressionReroute_21.InputChannel, B=@n15) {constB=1.5}))
  FunctionOutput(None=MakeMaterialAttributes!(
    BaseColor=NamedRerouteUsage()
    Metallic=NamedRerouteUsage()
    Specular=NamedRerouteUsage()
    Roughness=NamedRerouteUsage()
    Anisotropy
    EmissiveColor=NamedRerouteUsage()
    Opacity
    OpacityMask
    Normal=NamedRerouteUsage()
    Tangent
    WorldPositionOffset=NamedRerouteUsage()
    SubsurfaceColor
    ClearCoat
    ClearCoatRoughness
    AmbientOcclusion
    Refraction
    CustomizedUVs_0
    CustomizedUVs_1
    CustomizedUVs_2
    CustomizedUVs_3
    CustomizedUVs_4
    CustomizedUVs_5
    CustomizedUVs_6
    CustomizedUVs_7
    PixelDepthOffset
    ShadingModel
    Displacement
   )) {outputName=MaterialAttributes}
  n16 = NamedRerouteUsage()
  n17 = Multiply(
   A=StaticSwitchParameter(True=NamedRerouteUsage(), False=NamedRerouteUsage()) {group=Tire Color Control, parameterName=UseTireCustomColor}
   B=ScalarParameter() {defaultValue=1, group=Tire Color Control, parameterName=Tire_CustomColorBrightness}
  )
  MaterialExpressionReroute_13 = Reroute(None=@n5.R)
  n18 = OneMinus(None=@MaterialExpressionReroute_13.InputChannel)
  n19 = Saturate(None=Multiply(A=AppendVector(A=@n18, B=@MaterialExpressionReroute_13.InputChannel)) {constB=3})
  n20 = ComponentMask(None=@n19) {r=true}
  n21 = ComponentMask(None=@n19) {g=true}
  MaterialExpressionReroute_14 = Reroute(None=@n6)
  n22 = OneMinus(None=Saturate(None=LinearInterpolate(A=@n20, B=@n21, Alpha=@MaterialExpressionReroute_14.FloorValue)))
  MaterialExpressionReroute_31 = Reroute(None=@n22)
  n23 = Add(A=@MaterialExpressionReroute_15.InputChannel) {constB=-0.5}
  n24 = AppendVector(
   A=Multiply(A=Saturate(None=@n23)) {constB=2}
   B=Add(A=@MaterialExpressionReroute_15.InputChannel, B=@MaterialExpressionReroute_15.InputChannel)
  )
  n25 = MaterialFunctionCall(Float2=Saturate(None=Multiply(A=@n24, B=@n24))) {materialFunction=&BreakOutFloat2Components}
  n26 = LinearInterpolate(A=@n25.R, B=@n25.R, Alpha=@MaterialExpressionReroute_16.FloorValue)
  MaterialExpressionReroute_17 = Reroute(None=@n26)
  n27 = StaticSwitchParameter(True=@MaterialExpressionReroute_31.Output, False=@MaterialExpressionReroute_17.Output) {
   desc="Allows for a flip-flop version of the floored colorization; e.g. white value is one mask, black is the other when flipped."
   group=Tire Properties
   parameterName=Tire_FlooredCustomFlipFlop
  }
  n28 = StaticSwitchParameter(True=@n27, False=@n5.R) {
   desc=Allows for the mask to have a color range under pure white that can be made visible by adjusting the corresponding scalar parameter.
   group=Tire Properties
   parameterName=Tire_FlooredCustomColor
  }
  n29 = StaticSwitchParameter(True=OneMinus(None=@n28), False=@n28) {group=Tire Color Control, parameterName=Tire_InvertColorMask, sortPriority=2}
  n30 = LinearInterpolate(A=@n16, B=@n17, Alpha=@n29)
  n31 = Multiply(
   A=StaticSwitchParameter(True=NamedRerouteUsage(), False=NamedRerouteUsage()) {group=Tire Color Control, parameterName=UseTireAccentColor}
   B=ScalarParameter() {defaultValue=1, group=Tire Color Control, parameterName=Tire_AccentColorBrightness}
  )
  MaterialExpressionReroute_18 = Reroute(None=@n7.R)
  n32 = OneMinus(None=@MaterialExpressionReroute_18.InputChannel)
  n33 = Saturate(None=Multiply(A=AppendVector(A=@n32, B=@MaterialExpressionReroute_18.InputChannel)) {constB=3})
  n34 = ComponentMask(None=@n33) {r=true}
  n35 = ComponentMask(None=@n33) {g=true}
  MaterialExpressionReroute_29 = Reroute(None=@n8)
  n36 = OneMinus(None=Saturate(None=LinearInterpolate(A=@n34, B=@n35, Alpha=@MaterialExpressionReroute_29.FloorValue)))
  MaterialExpressionReroute_30 = Reroute(None=@n36)
  n37 = Add(A=@MaterialExpressionReroute_21.InputChannel) {constB=-0.5}
  n38 = AppendVector(
   A=Multiply(A=Saturate(None=@n37)) {constB=2}
   B=Add(A=@MaterialExpressionReroute_21.InputChannel, B=@MaterialExpressionReroute_21.InputChannel)
  )
  n39 = MaterialFunctionCall(Float2=Saturate(None=Multiply(A=@n38, B=@n38))) {materialFunction=&BreakOutFloat2Components}
  n40 = LinearInterpolate(A=@n39.R, B=@n39.R, Alpha=@MaterialExpressionReroute_25.FloorValue)
  MaterialExpressionReroute_26 = Reroute(None=@n40)
  n41 = StaticSwitchParameter(True=@MaterialExpressionReroute_30.Output, False=@MaterialExpressionReroute_26.Output) {
   desc="Allows for a flip-flop version of the floored colorization; e.g. white value is one mask, black is the other when flipped."
   group=Tire Properties
   parameterName=Tire_FlooredAccentFlipFlop
  }
  n42 = StaticSwitchParameter(True=@n41, False=@n7.R) {
   desc=Allows for the mask to have a color range under pure white that can be made visible by adjusting the corresponding scalar parameter.
   group=Tire Properties
   parameterName=Tire_FlooredAccentColor
  }
  n43 = StaticSwitchParameter(True=OneMinus(None=@n42), False=@n42) {group=Tire Color Control, parameterName=Tire_InvertAccentColorMask, sortPriority=2}
  n44 = StaticSwitchParameter(True=LinearInterpolate(A=@n30, B=@n31, Alpha=@n43), False=@n30) {group=Accent Color Properties, parameterName=EnableAccentColor, sortPriority=2}
  n45 = NamedRerouteUsage()
  n46 = LinearInterpolate(A=@n45, B=Multiply(A=@n17, B=@n29), Alpha=@n29)
  n47 = StaticSwitchParameter(True=LinearInterpolate(A=@n46, B=@n31, Alpha=@n43), False=@n46) {group=Accent Color Properties, parameterName=EnableAccentColor, sortPriority=2}
  NamedRerouteDeclaration(None=StaticSwitchParameter(
    True=StaticSwitchParameter(True=@n44, False=@n16) {group=Tire Color Control, parameterName=TireColorIsMasked}
    False=StaticSwitchParameter(True=@n47, False=@n45) {group=Tire Color Control, parameterName=TireColorIsMasked}
   ) {group=Tire Color Control, parameterName=ColorTireDiffuse}) {name=ColorTireDiffuse, nodeColor=<rgba 0.4297 0 1 1>}
  n48 = NamedRerouteUsage()
  n49 = StaticSwitchParameter(True=NamedRerouteUsage(), False=@n48) {
   desc="True will use RGB paramter as diffuse; False will use texture as diffuse"
   group=Rim Color Control
   parameterName=RimUseColorRGBAsDiffuse
  }
  MaterialExpressionReroute_39 = Reroute(None=@n9.R)
  n50 = OneMinus(None=@MaterialExpressionReroute_39.InputChannel)
  n51 = Saturate(None=Multiply(A=AppendVector(A=@n50, B=@MaterialExpressionReroute_39.InputChannel)) {constB=3})
  n52 = ComponentMask(None=@n51) {r=true}
  n53 = ComponentMask(None=@n51) {g=true}
  MaterialExpressionReroute_41 = Reroute(None=@n10)
  n54 = OneMinus(None=Saturate(None=LinearInterpolate(A=@n52, B=@n53, Alpha=@MaterialExpressionReroute_41.FloorValue)))
  MaterialExpressionReroute_42 = Reroute(None=@n54)
  MaterialExpressionReroute_44 = Reroute(None=@n9.R)
  n55 = Add(A=@MaterialExpressionReroute_44.InputChannel) {constB=-0.5}
  n56 = AppendVector(
   A=Multiply(A=Saturate(None=@n55)) {constB=2}
   B=Add(A=@MaterialExpressionReroute_44.InputChannel, B=@MaterialExpressionReroute_44.InputChannel)
  )
  n57 = MaterialFunctionCall(Float2=Saturate(None=Multiply(A=@n56, B=@n56))) {materialFunction=&BreakOutFloat2Components}
  MaterialExpressionReroute_45 = Reroute(None=@n10)
  n58 = LinearInterpolate(A=@n57.R, B=@n57.R, Alpha=@MaterialExpressionReroute_45.FloorValue)
  MaterialExpressionReroute_43 = Reroute(None=@n58)
  n59 = StaticSwitchParameter(True=@MaterialExpressionReroute_42.Output, False=@MaterialExpressionReroute_43.Output) {
   desc="Allows for a flip-flop version of the floored colorization; e.g. white value is one mask, black is the other when flipped."
   group=Rim Properties
   parameterName=Rim_FlooredCustomFlipFlop
  }
  n60 = StaticSwitchParameter(True=@n59, False=@n9.R) {
   desc=Allows for the mask to have a color range under pure white that can be made visible by adjusting the corresponding scalar parameter.
   group=Rim Properties
   parameterName=Rim_FlooredCustomColor
  }
  n61 = StaticSwitchParameter(True=OneMinus(None=@n60), False=@n60) {group=Rim Color Control, parameterName=Rim_InvertColorMask, sortPriority=2}
  n62 = Multiply(
   A=Multiply(A=NamedRerouteUsage(), B=ScalarParameter() {defaultValue=1, group=Rim Color Control, parameterName=Rim_CustomColorBrightness})
   B=@n61
  )
  n63 = LinearInterpolate(A=@n49, B=@n62, Alpha=@n61)
  n64 = NamedRerouteUsage()
  n65 = ScalarParameter() {defaultValue=1, group=Rim Color Control, parameterName=Rim_AccentColorBrightness}
  MaterialExpressionReroute_24 = Reroute(None=@n2.R)
  n66 = OneMinus(None=@MaterialExpressionReroute_24.InputChannel)
  n67 = Saturate(None=Multiply(A=AppendVector(A=@n66, B=@MaterialExpressionReroute_24.InputChannel)) {constB=3})
  n68 = ComponentMask(None=@n67) {r=true}
  n69 = ComponentMask(None=@n67) {g=true}
  MaterialExpressionReroute_33 = Reroute(None=@n3)
  n70 = OneMinus(None=Saturate(None=LinearInterpolate(A=@n68, B=@n69, Alpha=@MaterialExpressionReroute_33.FloorValue)))
  MaterialExpressionReroute_34 = Reroute(None=@n70)
  n71 = Add(A=@MaterialExpressionReroute_22.InputChannel) {constB=-0.5}
  n72 = AppendVector(
   A=Multiply(A=Saturate(None=@n71)) {constB=2}
   B=Add(A=@MaterialExpressionReroute_22.InputChannel, B=@MaterialExpressionReroute_22.InputChannel)
  )
  n73 = MaterialFunctionCall(Float2=Saturate(None=Multiply(A=@n72, B=@n72))) {materialFunction=&BreakOutFloat2Components}
  n74 = LinearInterpolate(A=@n73.R, B=@n73.R, Alpha=@MaterialExpressionReroute_23.FloorValue)
  MaterialExpressionReroute_12 = Reroute(None=@n74)
  n75 = StaticSwitchParameter(True=@MaterialExpressionReroute_34.Output, False=@MaterialExpressionReroute_12.Output) {
   desc="Allows for a flip-flop version of the floored colorization; e.g. white value is one mask, black is the other when flipped."
   group=Rim Properties
   parameterName=Rim_FlooredAccentFlipFlop
  }
  n76 = StaticSwitchParameter(True=@n75, False=@n2.R) {
   desc=Allows for the mask to have a color range under pure white that can be made visible by adjusting the corresponding scalar parameter.
   group=Rim Properties
   parameterName=Rim_FlooredAccentColor
  }
  n77 = StaticSwitchParameter(True=OneMinus(None=@n76), False=@n76) {group=Rim Color Control, parameterName=Rim_InvertAccentColorMask, sortPriority=2}
  n78 = StaticSwitchParameter(True=LinearInterpolate(A=@n63, B=Multiply(A=@n64, B=@n65), Alpha=@n77), False=@n63) {group=Accent Color Properties, parameterName=EnableAccentColor, sortPriority=2}
  n79 = StaticSwitchParameter(True=NamedRerouteUsage(), False=@n62) {desc=True will make the rim color utilize ONLY the RGB diffuse parameter, group=Rim Color Control, parameterName=RimOnlyRGBColor}
  NamedRerouteDeclaration(None=StaticSwitchParameter(True=StaticSwitchParameter(True=@n78, False=@n79) {group=Rim Color Control, parameterName=RimColorIsMasked}, False=@n48) {group=Rim Color Control, parameterName=ColorRimColorizeDiffuseTexture}) {name=ColorRimDiffuse, nodeColor=<rgba 0 0.8828 1 1>}
  n80 = NamedRerouteUsage()
  n81 = ScalarParameter() {defaultValue=1, parameterName=RimAOMultiplier}
  n82 = StaticSwitchParameter(True=NamedRerouteUsage(), False=Constant() {r=1}) {defaultValue=true, group=Tire Properties, parameterName=TireHasAO}
  n83 = ScalarParameter() {defaultValue=1, parameterName=TireAOMultiplier}
  n84 = LinearInterpolate(A=Saturate(None=Multiply(A=@n80, B=@n81)), B=Saturate(None=Multiply(A=@n82, B=@n83)), Alpha=NamedRerouteUsage())
  NamedRerouteDeclaration(None=@n84) {name=AOxDiffuse, nodeColor=<rgba 1 0.2344 0 1>}
  n85 = TextureCoordinate()
  n86 = AppendVector(
   A=LinearInterpolate(Alpha=ScalarParameter() {defaultValue=1, parameterName=TreadUVDirection_X}) {constA=-1}
   B=LinearInterpolate(Alpha=ScalarParameter() {defaultValue=1, parameterName=TreadUVDirection_Y}) {constA=-1}
  )
  n87 = TextureSampleParameter2D(UVs=NamedRerouteDeclaration(None=Multiply(A=@n85, B=@n86)) {name=TreadUVs, nodeColor=<rgba 0 1 0.4375 1>}) {group=Tire Properties, parameterName=TireDiffuse, texture=/VehicleCosmetics/Textures/T_White}
  NamedRerouteDeclaration(None=@n87.A) {name=TireDiffuse_A, nodeColor=<rgba 0 0.8125 1 1>}
  n88 = TextureSampleParameter2D() {group=Rim Properties, parameterName=RimDiffuse, texture=&T_White}
  NamedRerouteDeclaration(None=@n88.A) {name=RimeDiffuse_A, nodeColor=<rgba 1 0 0.8203 1>}
  n89 = StaticSwitchParameter(True=Max(A=@n61, B=@n77), False=@n61) {group=Accent Color Properties, parameterName=EnableAccentColor, sortPriority=2}
  NamedRerouteDeclaration(None=StaticSwitchParameter(True=StaticSwitchParameter(True=@n89, False=@n61) {group=Rim Color Control, parameterName=RimColorIsMasked}, False=@n61) {group=Rim Color Control, parameterName=ColorRimColorizeDiffuseTexture}) {name=Rim_InvertColorMask, nodeColor=<rgba 0.0312 1 0 1>}
  n90 = ChannelMaskParameter(None=@n1.RGBA) {group=Rim Color Control, maskChannel=Alpha, parameterName=Rim_AOChannelPicker, sortPriority=2, defaultValue=<rgba 0 0 0 1>}
  NamedRerouteDeclaration(None=@n90.R) {name=Rim_AOChannelPicker, nodeColor=<rgba 0 1 0.3438 1>}
  n91 = ChannelMaskParameter(None=@n4.A) {group=Tire Color Control, maskChannel=Alpha, parameterName=Tire_AOChannelPicker, sortPriority=2, defaultValue=<rgba 0 0 0 1>}
  NamedRerouteDeclaration(None=@n91.R) {name=Tire_CustomColorMask, nodeColor=<rgba 0 0.7188 1 1>}
  NamedRerouteDeclaration(None=@n88.RGB) {name=RimDiffuse_RGB, nodeColor=<rgba 0 0.9766 1 1>}
  NamedRerouteDeclaration(None=@n87.RGB) {name=TireDiffuse_RGB, nodeColor=<rgba 0 0.7422 1 1>}
  NamedRerouteDeclaration(None=VectorParameter() {parameterName=TireColor, defaultValue=<rgba 0.015 0.015 0.015 0>}.RGB) {name=TireColor_RGB, nodeColor=<rgba 0 1 0.8594 1>}
  NamedRerouteDeclaration(None=VectorParameter() {group=Custom Color Properties, parameterName=CustomColor, defaultValue=<rgba 1 1 1 0>}.RGB) {name=CustomColor_RGB, nodeColor=<rgba 0 1 0.5078 1>}
  NamedRerouteDeclaration(None=VertexColor().R) {name=VertColor_R, nodeColor=<rgba 0 0.4141 1 1>}
  NamedRerouteDeclaration(None=VectorParameter() {group=Custom Color Properties, parameterName=TireCustomColor, defaultValue=<rgba 1 1 1 0>}.RGB) {name=TireCustomColor_RGB, nodeColor=<rgba 0 0.0859 1 1>}
  NamedRerouteDeclaration(None=VectorParameter() {group=Accent Color Properties, parameterName=AccentColor, defaultValue=<rgba 1 1 1 0>}.RGB) {name=AccentColor_RGB, nodeColor=<rgba 0 0.6016 1 1>}
  NamedRerouteDeclaration(None=VectorParameter() {group=Accent Color Properties, parameterName=TireAccentColor, defaultValue=<rgba 1 1 1 0>}.RGB) {name=TireAccentColor_RGB, nodeColor=<rgba 1 0 0.3516 1>}
  n92 = TextureSampleParameter2D() {
   desc="R=Metallic, G=Rough, B=Spec, A=Unused"
   group=Textures
   parameterName=RimPBR
   samplerType=SAMPLERTYPE_LinearColor
   texture=/VehicleCosmetics/Textures/T_Black_Linear
  }
  NamedRerouteDeclaration(None=@n92.RGBA) {name=RimPBR, desc="R=Metallic, G=Rough, B=Spec, A=Unused", nodeColor=<rgba 0 1 0.1562 1>}
  n93 = TextureSampleParameter2D(UVs=NamedRerouteUsage()) {
   desc="R=Metallic, G=Rough, B=Spec, A=Unused"
   group=Textures
   parameterName=TirePBR
   samplerType=SAMPLERTYPE_LinearColor
   texture=&T_Black_Linear
  }
  NamedRerouteDeclaration(None=@n93.RGBA) {name=TirePBR, desc="R=Metallic, G=Rough, B=Spec, A=Unused", nodeColor=<rgba 0 1 0.1562 1>}
  NamedRerouteDeclaration(None=VectorParameter() {parameterName=RimColor, defaultValue=<rgba 1 0.8555 0 0>}.RGB) {name=RimColor_RGB, nodeColor=<rgba 1 0 0.3281 1>}
  n94 = Max(A=@n29, B=@n43)
  NamedRerouteDeclaration(None=StaticSwitchParameter(
    True=StaticSwitchParameter(True=@n94, False=@n29) {group=Accent Color Properties, parameterName=EnableAccentColor, sortPriority=2}
    False=@n29
   ) {group=Tire Color Control, parameterName=TireColorIsMasked}) {name=Tire_InvertColorMask, nodeColor=<rgba 1 0.6563 0 1>}
  NamedRerouteDeclaration(None=VectorParameter() {group=Rim Color Control, parameterName=Rim_EmissiveColor}.RGB) {name=RimEmissiveColor, nodeColor=<rgba 0.3125 1 0 1>}
  n95 = TextureSampleParameter2D() {group=Rim Properties, parameterName=RimNormal, samplerType=SAMPLERTYPE_Normal, texture=/VehicleCosmetics/Textures/T_BlankNormal}
  n96 = LinearInterpolate(Alpha=Saturate(None=ScalarParameter() {defaultValue=1, parameterName=TreadUVDirection_X})) {constA=-1}
  n97 = LinearInterpolate(Alpha=Saturate(None=ScalarParameter() {defaultValue=1, parameterName=TreadUVDirection_Y})) {constA=-1}
  n98 = TextureSampleParameter2D(UVs=Multiply(A=TextureCoordinate(), B=AppendVector(A=@n96, B=@n97))) {group=Tire Properties, parameterName=TireNormal, samplerType=SAMPLERTYPE_Normal, texture=&T_BlankNormal}
  n99 = Multiply(A=@n98.R, B=@n96) {constB=-1}
  n100 = Multiply(A=@n98.G, B=@n97) {constB=-1}
  NamedRerouteDeclaration(None=LinearInterpolate(
    A=@n95.RGB
    B=MaterialFunctionCall(R=@n99, G=@n100, B=@n98.B) {materialFunction=&AppendMany}.RGB
    Alpha=NamedRerouteUsage()
   )) {name=Normal_Out, nodeColor=<rgba 0.3125 1 0 1>}
  n101 = ChannelMaskParameter(None=TextureSampleParameter2D() {group=Rim Properties, parameterName=Rim_EmissiveMask, texture=&T_Black}.RGBA) {group=Rim Properties, parameterName=Rim_EmissiveChannelPicker, sortPriority=2}
  n102 = StaticSwitchParameter(True=NamedRerouteUsage(), False=NamedRerouteUsage()) {
   defaultValue=true
   desc=When true: We use the CustomColor_RGB for the emissive color. When false: Use the RimEmissiveColor parameter instead.
   group=Rim Properties
   parameterName=Rim_EmissiveUseCustomColor
  }
  n103 = Multiply(
   A=StaticSwitchParameter(True=OneMinus(None=@n101.R), False=@n101.R) {group=Rim Properties, parameterName=Rim_InvertEmissiveMask, sortPriority=2}
   B=Multiply(A=@n102, B=ScalarParameter() {defaultValue=1, group=Rim Properties, parameterName=Rim_EmissiveMultiplier, sortPriority=3})
  )
  n104 = Constant()
  n105 = ChannelMaskParameter(None=TextureSampleParameter2D() {group=Tire Properties, parameterName=Tire_EmissiveMask, texture=&T_Black}.RGBA) {group=Tire Properties, parameterName=Tire_EmissiveChannelPicker, sortPriority=2}
  n106 = NamedRerouteUsage()
  n107 = NamedRerouteUsage()
  n108 = Multiply(
   A=StaticSwitchParameter(True=OneMinus(None=@n105.R), False=@n105.R) {group=Tire Properties, parameterName=Tire_InvertEmissiveMask, sortPriority=2}
   B=Multiply(
    A=StaticSwitchParameter(True=@n106, False=@n107) {parameterName=UseTireCustomColor}
    B=ScalarParameter() {defaultValue=1, group=Tire Properties, parameterName=Tire_EmissiveMultiplier, sortPriority=3}
   )
  )
  n109 = Constant()
  NamedRerouteDeclaration(None=LinearInterpolate(
    A=StaticSwitchParameter(True=@n103, False=@n104) {group=Rim Properties, parameterName=UseRimEmissive, sortPriority=1}
    B=StaticSwitchParameter(True=@n108, False=@n109) {group=Tire Properties, parameterName=UseTireEmissive, sortPriority=1}
    Alpha=NamedRerouteUsage()
   )) {name=Emissive_Out, nodeColor=<rgba 1 0.1172 0 1>}
  n110 = VertexNormalWS()
  n111 = ScalarParameter() {parameterName=TirePop}
  n112 = Subtract(A=ObjectPositionWS(), B=WorldPosition().XYZ)
  n113 = VertexColor()
  NamedRerouteDeclaration(None=LinearInterpolate(A=Multiply(A=@n110) {constB=0}, B=Multiply(A=@n111, B=@n112), Alpha=Saturate(None=@n113.G))) {name=WPO_Out, nodeColor=<rgba 0.9688 0 1 1>}
  n114 = ComponentMask(None=NamedRerouteUsage()) {g=true}
  n115 = LinearInterpolate(
   A=ScalarParameter() {defaultValue=0.5, group=Rim Properties, parameterName=RimRoughness, sortPriority=1}
   B=ScalarParameter() {defaultValue=0.5, group=Rim Properties, parameterName=RimRoughness_Painted, sortPriority=1}
   Alpha=NamedRerouteUsage()
  )
  n116 = ComponentMask(None=NamedRerouteUsage()) {g=true}
  n117 = ScalarParameter() {defaultValue=0.85, group=Tire Properties, parameterName=TireRoughness, sortPriority=1}
  n118 = StaticSwitchParameter(
   True=LinearInterpolate(
    A=@n117
    B=ScalarParameter() {defaultValue=0.85, group=Tire Properties, parameterName=TireRoughness_Painted, sortPriority=1}
    Alpha=NamedRerouteUsage()
   )
   False=@n117
  ) {parameterName=TireRoughnessFromPainted}
  NamedRerouteDeclaration(None=LinearInterpolate(
    A=StaticSwitchParameter(True=@n114, False=@n115) {parameterName=RimRoughnessFromPBR}
    B=StaticSwitchParameter(True=@n116, False=@n118) {parameterName=TireRoughnessFromPBR}
    Alpha=NamedRerouteUsage()
   )) {name=Roughness_Out, nodeColor=<rgba 1 0.3047 0 1>}
  n119 = ComponentMask(None=NamedRerouteUsage()) {b=true}
  n120 = ScalarParameter() {defaultValue=1, group=Rim Properties, parameterName=RimSpecular, sortPriority=1}
  n121 = ComponentMask(None=NamedRerouteUsage()) {b=true}
  n122 = ScalarParameter() {group=Tire Properties, parameterName=TireSpecular, sortPriority=1}
  NamedRerouteDeclaration(None=LinearInterpolate(
    A=StaticSwitchParameter(True=@n119, False=@n120) {parameterName=RimSpecularFromPBR}
    B=StaticSwitchParameter(True=@n121, False=@n122) {parameterName=TireSpecularFromPBR}
    Alpha=NamedRerouteUsage()
   )) {name=Spec_Out, nodeColor=<rgba 0.2656 0 1 1>}
  n123 = ComponentMask(None=NamedRerouteUsage()) {r=true}
  n124 = LinearInterpolate(
   A=ScalarParameter() {group=Rim Properties, parameterName=RimMetallic, sortPriority=1}
   B=ScalarParameter() {defaultValue=1, group=Rim Properties, parameterName=RimMetallic_Painted, sortPriority=1}
   Alpha=NamedRerouteUsage()
  )
  n125 = ComponentMask(None=NamedRerouteUsage()) {r=true}
  n126 = ScalarParameter() {group=Tire Properties, parameterName=TireMetallic, sortPriority=1}
  n127 = StaticSwitchParameter(
   True=LinearInterpolate(A=@n126, B=ScalarParameter() {group=Tire Properties, parameterName=TireMetallic_Painted, sortPriority=1}, Alpha=NamedRerouteUsage())
   False=@n126
  ) {parameterName=TireMetallicFromPainted}
  NamedRerouteDeclaration(None=LinearInterpolate(
    A=StaticSwitchParameter(True=@n123, False=@n124) {parameterName=RimMetallicFromPBR}
    B=StaticSwitchParameter(True=@n125, False=@n127) {parameterName=TireMetallicFromPBR}
    Alpha=NamedRerouteUsage()
   )) {name=Metallic_Out, nodeColor=<rgba 0 1 0.6719 1>}
  n128 = NamedRerouteUsage()
  n129 = NamedRerouteUsage()
  n130 = NamedRerouteUsage()
  NamedRerouteDeclaration(None=Multiply(A=LinearInterpolate(A=@n128, B=@n129, Alpha=@n130), B=@n84)) {name=Diffuse_Out, nodeColor=<rgba 0 1 1 1>}
  NamedRerouteUsage()
  NamedRerouteUsage()
  NamedRerouteUsage()
  PinBase() {reroutePins [
    InputChannel {expression=&self:MaterialExpressionReroute_22}
    FloorValue {expression=&self:MaterialExpressionReroute_23}
   ]}
  PinBase!(Output=@n74) {reroutePins [Output {expression=&self:MaterialExpressionReroute_12}]}
  PinBase!() {reroutePins [
    InputChannel {expression=&self:MaterialExpressionReroute_18}
    FloorValue {expression=&self:MaterialExpressionReroute_29}
   ]}
  PinBase!(Output=@n36) {reroutePins [Output {expression=&self:MaterialExpressionReroute_30}]}
  PinBase!() {reroutePins [
    InputChannel {expression=&self:MaterialExpressionReroute_39}
    FloorValue {expression=&self:MaterialExpressionReroute_41}
   ]}
  PinBase!(Output=@n54) {reroutePins [Output {expression=&self:MaterialExpressionReroute_42}]}
  PinBase!() {reroutePins [
    InputChannel {expression=&self:MaterialExpressionReroute_44}
    FloorValue {expression=&self:MaterialExpressionReroute_45}
   ]}
  PinBase!(Output=@n58) {reroutePins [Output {expression=&self:MaterialExpressionReroute_43}]}
  PinBase!() {reroutePins [
    InputChannel {expression=&self:MaterialExpressionReroute_24}
    FloorValue {expression=&self:MaterialExpressionReroute_33}
   ]}
  PinBase!(Output=@n70) {reroutePins [Output {expression=&self:MaterialExpressionReroute_34}]}
  PinBase!() {reroutePins [
    InputChannel {expression=&self:MaterialExpressionReroute_15}
    FloorValue {expression=&self:MaterialExpressionReroute_16}
   ]}
  PinBase!(Output=@n26) {reroutePins [Output {expression=&self:MaterialExpressionReroute_17}]}
  PinBase!() {reroutePins [
    InputChannel {expression=&self:MaterialExpressionReroute_13}
    FloorValue {expression=&self:MaterialExpressionReroute_14}
   ]}
  PinBase!(Output=@n22) {reroutePins [Output {expression=&self:MaterialExpressionReroute_31}]}
  PinBase!() {reroutePins [
    InputChannel {expression=&self:MaterialExpressionReroute_21}
    FloorValue {expression=&self:MaterialExpressionReroute_25}
   ]}
  PinBase!(Output=@n40) {reroutePins [Output {expression=&self:MaterialExpressionReroute_26}]}
 ]}
