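# Dragon wheel material for skeletal vehicle wheels, adding an animated, color-tintable flame glow to the shared MF_WheelParent surface.
# It keeps the parent's base color, metallic, specular, roughness and normal, overriding only emissive.
# A panned flow texture's RG multiplies the flame mask UVs, combining with static and scrolling samples of that mask plus the rim mask, tinted by CustomColor.
# Vertex color red blends between the rim-masked glow and the layered tire flames.
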Material /VehicleCosmetics/Wheels/Dragon/Materials/M_Wheel_Dragon
graph {
 properties {
  bUseMaterialAttributes=true
  bUsedWithMeshDeformer=true
  bUsedWithSkeletalMesh=true
  pixelDepthOffsetMode=PDOM_Legacy
  usedShadingModels=Default Lit
 }
 nodes [
  n1 = TextureCoordinate() {uTiling=0.5, vTiling=0.3}
  n2 = ScalarParameter() {defaultValue=1, parameterName=UVDistort_Tile}
  n3 = TextureSample(UVs=Panner(Coordinate=Multiply(A=@n1, B=@n2)) {speedX=0.25, speedY=0.5}) {samplerType=SAMPLERTYPE_LinearColor, texture=/VehicleCosmetics/Wheels/Dragon/Textures/T_Flow_03}
  Add(
   A=TextureCoordinate()
   B=Multiply(A=ComponentMask(None=@n3.RGB) {g=true, r=true}, B=ScalarParameter() {defaultValue=0.5, parameterName=UVDistort_Amount})
  )
  n4 = GetMaterialAttributes(None=MaterialFunctionCall() {materialFunction=/VehicleCosmetics/Wheels/_SharedMaterials/MF_WheelParent}.MaterialAttributes) {attributeGetTypes [
    #1
    #2
    #3
    #4
    #5
   ]}
  n5 = TextureSampleParameter2D() {parameterName=Rim_CustomColorMask, texture=/VehicleCosmetics/Wheels/Dragon/Textures/T_Wheel_Dragon_M}
  n6 = VectorParameter() {parameterName=CustomColor, defaultValue=<rgba 1 0.4946 0 1>}
  n7 = Multiply(A=TextureCoordinate() {uTiling=0.5, vTiling=0.3}, B=ScalarParameter() {defaultValue=1, parameterName=UVDistort_Tile})
  n8 = ComponentMask(None=TextureSample(UVs=Panner(Coordinate=@n7) {speedX=0.25, speedY=0.5}) {samplerType=SAMPLERTYPE_LinearColor, texture=&T_Flow_03}.RGB) {g=true, r=true}
  n9 = ScalarParameter() {defaultValue=0.5, parameterName=UVDistort_Amount}
  n10 = TextureSampleParameter2D(UVs=Multiply(A=TextureCoordinate() {vTiling=4}, B=Multiply(A=@n8, B=@n9))) {parameterName=Tire_Mask, texture=/VehicleCosmetics/Wheels/Dragon/Textures/T_Tire_Flame_M}
  n11 = Multiply(A=@n6.RGB, B=@n10.R)
  n12 = Multiply(A=@n5.R, B=Multiply(A=@n11) {constB=5})
  n13 = TextureCoordinate() {uTiling=0.5}
  n14 = TextureSampleParameter2D(UVs=@n13) {parameterName=Tire_Mask, texture=&T_Tire_Flame_M}
  n15 = Add(A=@n14.R, B=@n14.R)
  n16 = OneMinus(None=@n15)
  n17 = Multiply(A=@n5.B, B=Multiply(A=Saturate(None=@n16), B=@n14.B))
  n18 = Add(A=@n11, B=Multiply(A=Multiply(A=@n6.RGB, B=@n17)) {constB=10})
  n19 = Multiply(
   A=@n6.RGB
   B=TextureSampleParameter2D(UVs=Panner(Coordinate=@n13) {speedX=0.1}) {parameterName=Tire_Mask, texture=&T_Tire_Flame_M}.R
  )
  n20 = Saturate(None=Subtract(A=Add(A=@n5.B, B=@n14.B), B=@n15))
  n21 = Multiply(A=Multiply(A=Multiply(A=@n19) {constB=3}, B=Multiply(A=@n10.B, B=@n20))) {constB=10}
  SetMaterialAttributes!(
   MaterialAttributes=@n4.MaterialAttributes
   Base Color=@n4.MaterialAttributes
   Metallic=@n4.MaterialAttributes
   Specular=@n4.MaterialAttributes
   Roughness=@n4.MaterialAttributes
   Emissive Color=LinearInterpolate(A=Multiply(A=@n12) {constB=3}, B=Add(A=@n18, B=@n21), Alpha=VertexColor().R) {constB=0}
   Normal=@n4.MaterialAttributes
  ) {attributeSetTypes [
    #1
    #2
    #3
    #4
    #6
    #5
   ]}
 ]
}
